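# Figment winter HLOD proxy material with Nanite support, shading flattened distant meshes with snow and an ice tint.
# Switchable baked textures or constants feed the attributes, with diffuse optionally grass-replaced, then multiplied by shared winter snow color.
# These pass through destructible damage, snow progression, tree snow and HLOD WPO functions, while MaterialProxyReplace supplies a simpler proxy version.
# Extras include camera-distance Z offset, normal flattening near water height and vertex-color-indexed instance hiding.
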Material /Figment_S07/Athena/HLODSetup/HLOD_Snow/BaseFlattenMaterialFigment_SimpleIce
graph {
 properties {
  bTangentSpaceNormal=false
  bUseMaterialAttributes=true
  bUsedWithNanite=true
  bUsedWithStaticLighting=true
  ditheredLODTransition=true
  pixelDepthOffsetMode=PDOM_Legacy
  usedShadingModels=Default Lit
 }
 nodes [
  n1 = ScalarParameter() {defaultValue=40, parameterName=Z_Offset}
  n2 = Normalize(VectorInput=WorldPosition() {worldPositionShaderOffset=WPT_CameraRelative}.XYZ)
  ComponentMask(None=Multiply(A=@n1, B=Multiply(A=@n2) {constB=-1})) {b=true}
  Constant() {r=1}
  n3 = NamedRerouteUsage()
  n4 = StaticBool()
  n5 = StaticBool() {value=true}
  n6 = TextureCoordinate() {coordinateIndex=1}
  n7 = TextureCoordinate() {coordinateIndex=4}
  n8 = StaticSwitchParameter(
   True=TextureCoordinate() {coordinateIndex=5}
   False=StaticSwitchParameter(True=TextureCoordinate() {coordinateIndex=6}, False=TextureCoordinate()) {parameterName=UseUV6}
  ) {parameterName=UseUV5}
  n9 = StaticSwitchParameter(
   True=TextureCoordinate() {coordinateIndex=2}
   False=StaticSwitchParameter(True=TextureCoordinate() {coordinateIndex=3}, False=StaticSwitchParameter(True=@n7, False=@n8) {parameterName=UseUV4}) {parameterName=UseUV3}
  ) {parameterName=UseUV2}
  n10 = NamedRerouteDeclaration(None=StaticSwitchParameter(True=StaticSwitchParameter(True=@n6, False=@n9) {parameterName=UseUV1}, False=TextureCoordinate()) {parameterName=UseCustomUV}) {name=HLOD-UV, nodeColor=<rgba 1 0 0.8203 1>}
  n11 = TextureSampleParameter2D(UVs=@n10) {
   parameterName=PackedTexture
   samplerSource=SSM_Clamp_WorldGroupSettings
   samplerType=SAMPLERTYPE_LinearColor
   texture=/Engine/EngineMaterials/BaseFlattenLinearColor
  }
  n12 = TextureSampleParameter2D(UVs=@n10) {
   parameterName=MetallicTexture
   samplerSource=SSM_Clamp_WorldGroupSettings
   samplerType=SAMPLERTYPE_LinearColor
   texture=&BaseFlattenLinearColor
  }
  n13 = ScalarParameter() {parameterName=MetallicConst}
  n14 = TextureSampleParameter2D(UVs=@n10) {
   parameterName=SpecularTexture
   samplerSource=SSM_Clamp_WorldGroupSettings
   samplerType=SAMPLERTYPE_LinearColor
   texture=&BaseFlattenLinearColor
  }
  n15 = ScalarParameter() {parameterName=SpecularConst}
  n16 = TextureSampleParameter2D(UVs=@n10) {
   parameterName=RoughnessTexture
   samplerSource=SSM_Clamp_WorldGroupSettings
   samplerType=SAMPLERTYPE_LinearColor
   texture=&BaseFlattenLinearColor
  }
  n17 = ScalarParameter() {defaultValue=1, parameterName=RoughnessConst}
  n18 = TextureSampleParameter2D(UVs=@n10) {parameterName=EmissiveTexture, samplerSource=SSM_Clamp_WorldGroupSettings, texture=/Engine/EngineMaterials/BaseFlattenEmissiveMap}
  n19 = StaticSwitchParameter(True=@n18.RGB, False=VectorParameter() {parameterName=EmissiveConst}.RGB) {parameterName=UseEmissive}
  n20 = ScalarParameter() {defaultValue=1, parameterName=EmissiveScale}
  n21 = Constant3Vector() {constant=<rgba 0 0 1 0>}
  n22 = TextureSampleParameter2D(UVs=NamedRerouteUsage()) {
   parameterName=NormalTexture
   samplerSource=SSM_Clamp_WorldGroupSettings
   samplerType=SAMPLERTYPE_Normal
   texture=&FlatNormalBlue
  }
  n23 = Transform(None=@n22.RGB)
  n24 = WorldPosition()
  n25 = Abs(None=Subtract(A=ComponentMask(None=@n24.XYZ) {b=true}, B=ScalarParameter() {defaultValue=-754, parameterName=WaterZ}) {constB=-754})
  n26 = ScalarParameter() {defaultValue=150, parameterName=FlatRange}
  n27 = Saturate(None=Divide(A=Subtract(A=@n25, B=@n26), B=@n26))
  n28 = NamedRerouteUsage()
  n29 = Multiply(
   A=MaterialFunctionCall(
    Fade Length=ScalarParameter() {defaultValue=2000, parameterName=Z_Offset_FadeLength}
    Fade Offset=ScalarParameter() {defaultValue=50000, parameterName=Z_Offset_FadeOffset}
    For Vertex Shader=StaticBool() {value=true}
   ) {materialFunction=&CameraDepthFade}.Result
   B=AppendVector(A=Constant2Vector(), B=@n1)
  )
  n30 = TextureSampleParameter2D(UVs=@n10) {
   parameterName=OpacityMaskTexture
   samplerType=SAMPLERTYPE_LinearGrayscale
   texture=/Engine/EngineMaterials/BaseFlattenGrayscaleMap.BaseFlattenGrayScaleMap
  }
  n31 = TextureSampleParameter2D(UVs=@n10) {
   parameterName=AmbientOcclusionTexture
   samplerType=SAMPLERTYPE_LinearGrayscale
   texture=&BaseFlattenGrayScaleMap
  }
  n32 = TextureSampleParameter2D(UVs=@n10) {
   parameterName=OpacityTexture
   samplerType=SAMPLERTYPE_LinearGrayscale
   texture=&BaseFlattenGrayScaleMap
  }
  n33 = SetMaterialAttributes!(
   MaterialAttributes
   Base Color=NamedRerouteUsage()
   Metallic=StaticSwitchParameter(True=@n11.R, False=StaticSwitchParameter(True=@n12.RGB, False=@n13) {parameterName=UseMetallic}) {parameterName=PackMetallic}
   Specular=StaticSwitchParameter(True=@n11.B, False=StaticSwitchParameter(True=@n14.RGB, False=@n15) {parameterName=UseSpecular}) {parameterName=PackSpecular}
   Roughness=StaticSwitchParameter(True=@n11.G, False=StaticSwitchParameter(True=@n16.RGB, False=@n17) {parameterName=UseRoughness}) {parameterName=PackRoughness}
   Emissive Color=MaterialFunctionCall!(Realtime=Multiply(A=@n19, B=@n20)) {materialFunction=/Game/Athena/HLODSetup/Materials/MF_HLODProxyColorSpaceCorrection}.Result
   Opacity=StaticSwitchParameter(True=@n32.RGB, False=ScalarParameter() {defaultValue=1, parameterName=OpacityConst}) {parameterName=UseOpacity}
   Opacity Mask=StaticSwitchParameter(True=@n30.RGB, False=ScalarParameter() {defaultValue=1, parameterName=OpacityMaskConst}) {parameterName=UseOpacityMask}
   Normal=StaticSwitchParameter(True=LinearInterpolate(A=@n21, B=@n23, Alpha=@n27), False=@n23) {parameterName=UseWorldZNormalFlattening}
   World Position Offset=StaticSwitchParameter(True=Add(A=@n28, B=@n29), False=@n28) {defaultValue=true, parameterName=UseZOffset}
   Ambient Occlusion=StaticSwitchParameter(True=@n31.RGB, False=ScalarParameter() {defaultValue=1, parameterName=AmbientOcclusionConst}) {parameterName=UseAmbientOcclusion}
  ) {attributeSetTypes [
    #1
    #2
    #3
    #4
    #5
    #6
    #7
    #8
    #9
    #10
   ]}
  n34 = StaticBool() {value=true}
  n35 = StaticBool()
  n36 = StaticBool() {value=true}
  n37 = StaticBool()
  n38 = NamedRerouteUsage()
  n39 = MaterialFunctionCall!(
   IsStaticSnow
   UseAlternateStaticSnowMask
   AlternateStaticSnowMask
   "UseRoad&&OcclusionMask"=StaticBool() {value=true}
   IsForGrassMeshToPushDown
   AdditionalMaskToRemoveSnow
   IsFallingSnowFX
   UseMinimapCoords
   IncreaseSnowUsingSlope
   UseSmoothBorder
   LerpHighM
   TemporalHeightLerpContrast
   IsLobbyIsland
   UseBR-IslandMask
   IsTree
   IsForRenderTarget
   UseSparkle
   UseSnowNormals
   UseAdditionalMaskToRemoveSnow
   IsTangentSpaceInput
   UseVertexNormalsForBasicSnow
   UseTangentSnowNormals
   Material=MaterialFunctionCall(
    Material_In=@n33
    UseCrackedDamageTextures=StaticSwitchParameter(True=@n34, False=@n35) {parameterName=UseDamageEffect}
    UseDamageWobbleWPO=StaticSwitchParameter(True=@n36, False=@n37) {parameterName=UseWobbleEffect}
    UseAdditiveWPO=StaticBool() {value=true}
    DestructibleDamageAmount=OneMinus(None=@n38)
    WSDeformationCenter=VectorParameter() {group=z_GameplayParameters_DestructibleDamage, parameterName=WSDeformationCenter}.RGB
    WSDamageRadius=ScalarParameter() {defaultValue=1, group=z_GameplayParameters_DestructibleDamage, parameterName=WSDamageRadius}
   ) {materialFunction=&MF_FN_DestructibleDamage}.Material_Out
  ) {materialFunction=/Game/Athena/Apollo/Environments/Landscape/Materials/Biome/UniqueBiomeMaterialFunctions/MF_Apollo_Snow_Progression}
  MaterialProxyReplace(
   Realtime=MaterialFunctionCall!(
    IsOcean
    IsTerrain
    In=MaterialFunctionCall!(
     ExtraSnowMask=@n3
     Simple Mask=@n4
     ForceUseInWSNormal
     UseActorPosition
     UseSnowMaskNormal=@n5
     UseColorTest
     UseColorTestRock
     In=@n39.Material
    ) {materialFunction=/Game/Packages/Fortnite_UniqueMaterials_1/MaterialFunctions/Trees/MF_Tree_Snow_Legacy}.FullyBlendedMA
   ) {materialFunction=/Game/Packages/Fortress_SharedMaterials/Base_Material_Functions/MF_HLODWPO}.Result
   MaterialProxy=@n33
  )
  n40 = Multiply(A=NamedRerouteUsage(), B=ScalarParameter() {defaultValue=2, parameterName=Rmult}) {constB=0.5}
  n41 = Multiply(A=NamedRerouteUsage(), B=ScalarParameter() {defaultValue=2, parameterName=Bmult}) {constB=0.5}
  Max(A=@n40, B=@n41)
  n42 = TextureSampleParameter2D(UVs=NamedRerouteUsage()) {parameterName=DiffuseTexture, samplerSource=SSM_Clamp_WorldGroupSettings, texture=/Engine/EngineMaterials/BaseFlattenDiffuseMap}
  NamedRerouteDeclaration(None=@n42.RGB) {name=DiffuseTexture, nodeColor=<rgba 1 0 0.8203 1>}
  NamedRerouteDeclaration(None=@n42.R) {name=Diffuse-R, nodeColor=<rgba 1 0 0.8203 1>}
  NamedRerouteDeclaration(None=@n42.G) {name=Diffuse-G, nodeColor=<rgba 1 0 0.8203 1>}
  NamedRerouteDeclaration(None=@n42.B) {name=Diffuse-B, nodeColor=<rgba 1 0 0.8203 1>}
  n43 = NamedRerouteUsage()
  n44 = ScalarParameter() {group=GrassReplacement, parameterName=GrassReplacement-GreenBias}
  n45 = DotProduct(A=ScalarParameter() {group=GrassReplacement, parameterName=GrassReplacement-S}, B=NamedRerouteUsage())
  n46 = Saturate(None=Divide(A=Add(A=@n43, B=@n44) {constB=0.01}, B=Saturate(None=@n45)))
  n47 = Max(A=NamedRerouteUsage(), B=NamedRerouteUsage())
  n48 = Divide(
   A=Subtract(A=Subtract(A=@n46, B=@n47), B=ScalarParameter() {defaultValue=0.99, group=GrassReplacement, parameterName=GrassReplacement-ReplaceMatchMin})
   B=ScalarParameter() {defaultValue=0.01, group=GrassReplacement, parameterName=GrassReplacement-ReplaceMatchLen}
  )
  n49 = Floor(None=Add(A=VertexColor().R) {constB=0})
  n50 = LinearInterpolate(
   A=NamedRerouteUsage()
   B=MaterialFunctionCall!(CombinedLandscapeLayersMask, BypassUseRVT, BypassBiomeColors) {materialFunction=/Game/Athena/Apollo/Environments/Landscape/Materials/Biome/CommonMaterials/MF_Common_Grass}.BasicGrassMeshColor
   Alpha=Multiply(A=Saturate(None=@n48), B=Saturate(None=@n49))
  )
  n51 = NamedRerouteUsage()
  n52 = StaticSwitchParameter(
   True=MaterialProxyReplace(Realtime=StaticSwitchParameter(True=@n50, False=@n51) {parameterName=UseGrassReplacement}, MaterialProxy=@n51)
   False=VectorParameter() {parameterName=DiffuseConst}.RGB
  ) {parameterName=UseDiffuse}
  n53 = MaterialFunctionCall() {materialFunction=/Figment_S07/LM_7_00/Athena/WinterArt/Materials/MF_SharedWinter}
  n54 = LinearInterpolate(B=@n53.SnowColor, Alpha=ScalarParameter() {parameterName=UseTimeOfDayBasedIceColor, sliderMax=1}) {constA=1, constAlpha=0}
  n55 = QualitySwitch(Default=@n53.SnowColor, Low=@n54, Medium=@n54)
  NamedRerouteDeclaration(None=MaterialProxyReplace(Realtime=Multiply(A=@n52, B=@n55), MaterialProxy=@n52)) {name=DiffuseA, nodeColor=<rgba 1 0 0.8203 1>}
  n56 = Add(A=ScalarParameter() {parameterName=GreenBias}, B=NamedRerouteUsage()) {constB=0.01}
  n57 = Saturate(None=DotProduct(A=ScalarParameter() {parameterName=s}, B=NamedRerouteUsage()))
  n58 = Subtract(A=Saturate(None=Divide(A=@n56, B=@n57)), B=@n41)
  n59 = Subtract(A=Saturate(None=Subtract(A=@n58, B=@n40)), B=ScalarParameter() {defaultValue=0.99, parameterName=ReplaceMatchMin})
  n60 = ScalarParameter() {defaultValue=0.01, parameterName=ReplaceMatchLen}
  NamedRerouteDeclaration(None=Saturate(None=Divide(A=@n59, B=@n60))) {name=SnowMask, nodeColor=<rgba 0 0.5781 1 1>}
  n61 = VertexColor()
  n62 = Constant() {r=255}
  n63 = Add(A=Multiply(A=Multiply(A=@n61.G, B=@n62), B=Constant() {r=256}), B=Multiply(A=@n61.B, B=@n62))
  n64 = Divide(A=Constant() {r=1}, B=ScalarParameter() {parameterName=NumInstances})
  n65 = Multiply(A=Subtract(A=@n63), B=@n64)
  n66 = Multiply(A=@n64) {constB=0.5}
  n67 = TextureSampleParameter2D!(
   UVs=MaterialFunctionCall(X=Add(A=@n65, B=@n66), Y=Constant() {r=0.5}) {materialFunction=&MakeFloat2}.Result
   Level
   Apply View MipBias
  ) {
   constMipValue=0
   mipValueMode=TMVM_MipLevel
   parameterName=InstanceVisibilityTexture
   texture=&T_WhitesRGB
  }
  n68 = NamedRerouteDeclaration(None=@n67.R) {name=InstanceVisibilityTexture-R, nodeColor=<rgba 1 0 0.8203 1>}
  n69 = StaticSwitch(
   True=Add(A=Ceil(None=@n68), B=Constant() {r=-1})
   False=Constant()
   Value=StaticBoolParameter() {defaultValue=true, parameterName=EnableInstanceDestroying}
  )
  n70 = Clamp(None=@n63)
  NamedRerouteDeclaration(None=Multiply(A=WorldPosition() {worldPositionShaderOffset=WPT_ExcludeAllShaderOffsets}.XYZ, B=Multiply(A=@n69, B=@n70)) {constB=-1}) {name=DestroyedInstance, nodeColor=<rgba 1 0 0.8203 1>}
  NamedRerouteUsage()
  n71 = Constant() {r=255}
  n72 = Divide(A=Constant() {r=1}, B=Constant() {r=4})
  n73 = Multiply(A=Clamp(None=Multiply(B=@n71), Min=Constant(), Max=Constant() {r=3}), B=@n72) {constB=0.25}
  n74 = Multiply(A=@n72) {constB=0.5}
  n75 = TextureSample(UVs=MaterialFunctionCall(X=Add(A=@n73, B=@n74), Y=Constant() {r=0.5}) {materialFunction=&MakeFloat2}.Result) {desc="This was once MaterialTexture, but that asset was never submitted"}
  n76 = OneMinus(None=MaterialFunctionCall(f3=@n75.RGB) {materialFunction=&AddComponents}.f3)
  n77 = Constant3Vector() {constant=<rgba 1 1 1 0>}
  n78 = Add(
   A=Multiply(A=ComponentMask(None=@n75.RGB) {r=true}, B=Constant3Vector() {constant=<rgba 1 0 0 0>})
   B=Multiply(A=ComponentMask(None=@n75.RGB) {g=true}, B=Constant3Vector() {constant=<rgba 0 1 0 0>})
  )
  n79 = Multiply(A=ComponentMask(None=@n75.RGB) {b=true}, B=Constant3Vector() {constant=<rgba 0 0 1 0>})
  StaticSwitchParameter(True=Add(A=Multiply(A=@n76, B=@n77), B=Add(A=@n78, B=@n79)), False=@n77) {parameterName=UseResourceType}
 ]
}
